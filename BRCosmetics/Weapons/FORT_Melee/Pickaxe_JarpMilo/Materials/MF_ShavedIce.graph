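# Material function that gives the JarpMilo pickaxe a layered shaved-ice look on top of input material attributes.
# Three tiled normal layers, two pushed inward by BumpOffset parallax and separated by mask channels, form the blended ice normal.
# Per-layer colors use light-facing ramps, Fresnel and a noise cover mask to build emissive shading, compensated for eye adaptation and time of day.
# Anime-style helpers derive the light vector and a Blinn/Phong highlight.
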MaterialFunction /BRCosmetics/Weapons/FORT_Melee/Pickaxe_JarpMilo/Materials/MF_ShavedIce
graph {nodes [
  AtmosphericLightVector()
  Custom() {code="return View.DirectionalLightDirection;", description=DirectionalLightVector, inputs []}
  n1 = GetMaterialAttributes(None=FunctionInput() {inputType=FunctionInput_MaterialAttributes}) {attributeGetTypes [#1, #2, #3]}
  n2 = NamedRerouteUsage()
  n3 = ScalarParameter() {defaultValue=0.1, group=ColorShading, parameterName=BaseColorIntensityOverride}
  n4 = NamedRerouteUsage()
  n5 = Multiply(A=@n1.EmissiveColor, B=ScalarParameter() {defaultValue=0.5, group=ColorShading, parameterName=EmissiveColorIntensityOverride})
  n6 = NamedRerouteUsage()
  FunctionOutput(None=SetMaterialAttributes!(
    MaterialAttributes=@n1.MaterialAttributes
    BaseColor=Multiply(A=@n2, B=@n3) {constB=0.2}
    Normal=MaterialFunctionCall(BaseNormal=@n1.Normal, AdditionalNormal=@n4) {materialFunction=&BlendAngleCorrectedNormals}.Result
    EmissiveColor=Add(A=@n5, B=@n6)
   ) {attributeSetTypes [#1, #2, #3]})
  n7 = TextureCoordinate()
  n8 = Multiply(A=@n7, B=ScalarParameter() {defaultValue=5, group=NormalBlending, parameterName=FirstLayerTiling}) {constB=7}
  n9 = TextureSampleParameter2D(UVs=@n8) {
   group=NormalBlending
   parameterName=FisrtLayerN
   samplerType=SAMPLERTYPE_Normal
   texture=/BRCosmetics/Weapons/FORT_Melee/Pickaxe_JarpMilo/Textures/T_ShavedIce_FirstN
  }
  NamedRerouteDeclaration(None=@n9.RGB) {name=FirstN_Tangent, nodeColor=<rgba 1 0.6094 0 1>}
  NamedRerouteDeclaration(None=@n1.BaseColor) {name=BaseColorIN, nodeColor=<rgba 1 0 0.4453 1>}
  n10 = LinearInterpolate(Alpha=NamedRerouteUsage()) {constA=1, constB=2}
  n11 = VectorParameter() {group=ColorShading, parameterName=ThirdLayerColor_Dark, defaultValue=<rgba 0 0.2095 0.4896 1>}
  n12 = Multiply(A=@n11.RGB, B=@n11.RGB)
  n13 = VectorParameter() {group=ColorShading, parameterName=ThirdLayerColor, sortPriority=3, defaultValue=<rgba 0 0.2377 0.3281 1>}
  n14 = Multiply(A=@n13.RGB, B=@n13.RGB)
  n15 = Fresnel(Normal=Transform(None=NamedRerouteUsage())) {baseReflectFraction=0, exponent=1}
  n16 = Multiply(
   A=LinearInterpolate(A=ScalarParameter() {defaultValue=0.2, group=ColorShading, parameterName=BackSideIntensity_Large}, Alpha=NamedRerouteUsage())
   B=StaticSwitchParameter(True=NamedRerouteUsage(), False=LinearInterpolate(A=@n12, B=@n14, Alpha=@n15)) {defaultValue=true, group=ColorShading, parameterName=UseOGcolor}
  )
  n17 = Power(Base=@n15, Exp=ScalarParameter() {defaultValue=3, group=ColorShading, parameterName=FresnelPow})
  n18 = VectorParameter() {group=ColorShading, parameterName=FresnelColor, sortPriority=3, defaultValue=<rgba 0.7585 0.8844 0.9323 1>}
  n19 = Multiply(A=@n18.RGB, B=@n18.RGB)
  NamedRerouteDeclaration(None=Add(A=Multiply(A=@n10, B=@n16), B=Multiply(A=@n17, B=@n19))) {name=ThirdLayerColor, nodeColor=<rgba 0 1 0.0625 1>}
  n20 = TextureCoordinate()
  n21 = TextureSampleParameter2D(UVs=BumpOffset(Coordinate=Multiply(A=@n20) {constB=3}, Height=Constant() {r=-20})) {
   parameterName=ThirdLayerN
   samplerType=SAMPLERTYPE_Normal
   texture=/BRCosmetics/Weapons/FORT_Melee/Pickaxe_JarpMilo/Textures/T_ShavedIce_ThirdN
  }
  n22 = TextureObjectParameter() {parameterName=LayerMask, samplerType=SAMPLERTYPE_Masks, texture=/BRCosmetics/Weapons/FORT_Melee/Pickaxe_JarpMilo/Textures/T_ShavedIce_Mask}
  n23 = NamedRerouteDeclaration(None=@n22) {name=LayerMask, nodeColor=<rgba 0.6406 0 1 1>}
  n24 = TextureSample(UVs=@n8, Tex=@n23) {samplerSource=SSM_Wrap_WorldGroupSettings, samplerType=SAMPLERTYPE_Masks}
  n25 = BumpOffset(
   Coordinate=Multiply(A=@n7) {constB=5}
   Height=Multiply(A=OneMinus(None=@n24.R), B=ScalarParameter() {defaultValue=-10, parameterName=SecondLayerBumpOffset})
  )
  n26 = TextureSampleParameter2D(UVs=@n25) {
   parameterName=SecondLayerN
   samplerType=SAMPLERTYPE_Normal
   texture=/BRCosmetics/Weapons/FORT_Melee/Pickaxe_JarpMilo/Textures/T_ShavedIce_SecondN
  }
  n27 = NamedRerouteUsage()
  n28 = NamedRerouteDeclaration(None=TextureSample(UVs=@n25, Tex=@n23) {samplerSource=SSM_Wrap_WorldGroupSettings, samplerType=SAMPLERTYPE_Masks}.G) {name=SecondArea, nodeColor=<rgba 0.9922 1 0 1>}
  n29 = LinearInterpolate(
   A=@n21.RGB
   B=LinearInterpolate(
    A=NamedRerouteDeclaration(None=@n26.RGB) {name=SecondLayerN, nodeColor=<rgba 0.0312 1 0 1>}
    B=NamedRerouteUsage()
    Alpha=NamedRerouteUsage()
   )
   Alpha=Saturate(None=Add(A=@n27, B=@n28))
  )
  NamedRerouteDeclaration(None=@n29) {name=Normal_OUT, nodeColor=<rgba 0 1 0.9297 1>}
  n30 = NamedRerouteUsage()
  n31 = NamedRerouteUsage()
  n32 = Fresnel(Normal=NamedRerouteUsage()) {baseReflectFraction=0}
  n33 = ScalarParameter() {defaultValue=1, group=ColorShading, parameterName=FresnelIntensity}
  n34 = NamedRerouteUsage()
  n35 = Add(
   A=Add(A=Add(A=@n30, B=@n31), B=NamedRerouteUsage())
   B=Multiply(A=Multiply(A=@n32, B=@n33) {constB=0.5}, B=LinearInterpolate(Alpha=@n34) {constA=0.2})
  )
  n36 = EyeAdaptation()
  n37 = Constant() {r=0.5}
  n38 = Constant() {r=1}
  n39 = Constant() {r=0.4}
  n40 = Constant() {r=0.2}
  NamedRerouteDeclaration(None=Multiply(
    A=Divide(A=@n35, B=@n36.EyeAdaptation)
    B=MaterialFunctionCall(Value_Morning=@n37, Value_Day=@n38, Value_Evening=@n39, Value_Night=@n40) {materialFunction=&Scalar_Time_of_Day}.Result
   )) {name=EmissiveOUT, nodeColor=<rgba 1 0.6094 0 1>}
  NamedRerouteDeclaration(None=Transform(None=@n29)) {name=Normal_WS, nodeColor=<rgba 0.2083 0.2083 1 1>}
  n41 = Custom() {code="return View.DirectionalLightDirection;", description=DirectionalLightVector, inputs []}
  n42 = DotProduct(A=@n41, B=@n41)
  n43 = CollectionParameter() {
   collection=&FortniteMaterialParameters
   parameterName=SunlightDirectionalVector
  }
  n44 = ComponentMask(None=@n43) {b=true, g=true, r=true}
  n45 = MaterialFunctionCall!(In=If(
    A=Custom() {code="return View.DirectionalLightColor;", description=DirectionalLightColor, inputs []}
    "A > B"=If(A=@n42, "A > B"=@n41, "A == B"=@n44, "A < B"=@n44) {desc=check if DirectionalLight exists}
    "A == B"=@n44
    "A < B"=@n44
   ) {desc=check if DirectionalLight color is Black}) {materialFunction=/Game/Packages/Fortress_SharedMaterials/Cinematics/Anime/MF_CINE_AnimeLightDirOverride}
  n46 = MaterialFunctionCall(Float3=@n45.Result) {materialFunction=&BreakOutFloat3Components}
  n47 = ScalarParameter() {
   desc="0.0 = the light vector is locked on the XY plane.\r\n1.0 = the actual light vector is used  \r\n(for Faces use a blend value of 0,0)"
   group=02_Anime Parameters
   parameterName=LockedZLightVector_Blend
   sliderMax=1
  }
  n48 = MaterialFunctionCall(X=@n46.R, Y=@n46.R, Z=Multiply(A=@n46.B, B=@n47)) {materialFunction=&MakeFloat3}
  n49 = StaticSwitchParameter(True=LinearInterpolate(A=@n45.Result, B=Normalize(VectorInput=@n48.Result), Alpha=VertexColor().A), False=@n45.Result) {
   desc="Lock the Z axis of the Light Vector\r\nis driven by the VertexColor A channel (0 = Regular LightVector, 1= Locked LIghtVector)"
   group=02_Anime Parameters
   parameterName=UseLockedZLightVector
   sortPriority=0
  }
  NamedRerouteDeclaration(None=@n49) {name=LightVector, nodeColor=<rgba 1 0.934 0.5103 1>}
  n50 = MaterialFunctionCall!(In=DotProduct(A=NamedRerouteUsage(), B=NamedRerouteUsage())) {materialFunction=/Game/Packages/Fortress_SharedMaterials/Cinematics/Anime/MF_CINE_AnimeLightingFeatures}
  NamedRerouteDeclaration(None=@n50.CINE_ShadingCurve) {name=NdotL, nodeColor=<rgba 0.9723 1 0.2825 1>}
  NamedRerouteDeclaration(None=DotProduct(A=NamedRerouteUsage(), B=CameraVectorWS())) {name=NdotV, nodeColor=<rgba 1 0.6242 0 1>}
  NamedRerouteDeclaration(None=@n21.RGB) {name=ThirdLayerN, nodeColor=<rgba 1 0 0.3516 1>}
  NamedRerouteDeclaration(None=@n50.CINE_LightSources) {name=CINE_LightSources, nodeColor=<rgba 1 0.9609 0 1>}
  NamedRerouteDeclaration(None=@n50.CINE_LightColor) {name=CINE_LightColor, nodeColor=<rgba 1 0.6803 0 1>}
  n51 = NamedRerouteUsage()
  n52 = Add(A=CameraVectorWS(), B=@n51)
  n53 = ScalarParameter() {defaultValue=1, group=02_Anime Parameters, parameterName=SpecularHighlight_LockedZVector, sliderMax=1}
  n54 = AppendVector(A=Constant2Vector() {g=1, r=1}, B=Saturate(None=OneMinus(None=@n53)))
  n55 = DotProduct(A=Normalize(VectorInput=Multiply(A=@n52, B=@n54)), B=NamedRerouteUsage()) {desc=Blinn}
  n56 = DotProduct(A=Normalize(VectorInput=Multiply(A=@n51, B=@n54)), B=ReflectionVectorWS(CustomWorldNormal=NamedRerouteUsage())) {desc=Phong}
  n57 = Saturate(None=LinearInterpolate(
    A=StaticSwitchParameter(True=@n55, False=@n56) {desc=Default is the Phong Highlight, group=02_Anime Parameters, parameterName=UseBlinnHighlight, sortPriority=0}
    B=NamedRerouteUsage()
    Alpha=ScalarParameter() {group=02_Anime Parameters, parameterName=SpecularHighlight_CameraFacing_Blend, sliderMax=1}
   ))
  n58 = Constant() {r=1}
  n59 = ScalarParameter() {defaultValue=0.5, group=02_Anime Parameters, parameterName=SpecularHighlight_Size, sliderMax=1, sliderMin=0.0001}
  n60 = ScalarParameter() {defaultValue=0.3, group=02_Anime Parameters, parameterName=SpecularHighlight_Hardness, sliderMax=1}
  NamedRerouteDeclaration(None=Multiply(
    A=SphereMask(A=@n57, B=@n58, Radius=@n59, Hardness=@n60) {hardnessPercent=78}
    B=ScalarParameter() {defaultValue=1, group=02_Anime Parameters, parameterName=SpecularHighlight_Intensity, sliderMax=8}
   )) {name=Specular_Curve, nodeColor=<rgba 1 0.4453 0 1>}
  NamedRerouteDeclaration(None=@n24.R) {name=FirstArea, nodeColor=<rgba 0 0.6719 1 1>}
  n61 = TextureSampleParameter2D(UVs=Multiply(A=TextureCoordinate(), B=ScalarParameter() {defaultValue=1, group=UV, parameterName=CoverTiling})) {
   group=ColorShading
   parameterName=CoverMask
   samplerType=SAMPLERTYPE_LinearColor
   texture=/Game/Global/Textures/Default/Noise/T_Noise01_Inverse
  }
  NamedRerouteDeclaration(None=@n61.R) {name=CoverMask, nodeColor=<rgba 0.7578 1 0 1>}
  n62 = Distance(A=@n29, B=Constant3Vector() {constant=<rgba 0 0 1 0>})
  NamedRerouteDeclaration(None=Power(Base=OneMinus(None=@n62)) {constExponent=20}) {name=HighLightGradient, nodeColor=<rgba 0.0547 0 1 1>}
  n63 = VectorParameter() {group=ColorShading, parameterName=FistLayerColor, defaultValue=<rgba 0.6875 0.8865 1 1>}
  n64 = DotProduct(A=NamedRerouteUsage(), B=PixelNormalWS())
  n65 = LinearInterpolate(
   A=ScalarParameter() {defaultValue=0.05, group=ColorShading, parameterName=BackSideIntensity_01}
   B=ScalarParameter() {defaultValue=1, group=ColorShading, parameterName=FrontSideIntensity_01}
   Alpha=NamedRerouteDeclaration(None=Saturate(None=@n64)) {name=SimpleShadingCurve, nodeColor=<rgba 0.4766 1 0 1>}
  ) {constA=0.2}
  n66 = Saturate(None=Distance(A=NamedRerouteUsage(), B=Constant3Vector() {constant=<rgba 0 0 1 0>}))
  n67 = Power(Base=@n66, Exp=ScalarParameter() {defaultValue=1.5, group=ColorShading, parameterName=FirstLayerIcePower}) {constExponent=1}
  n68 = OneMinus(None=@n66)
  n69 = Multiply(A=Multiply(A=Power(Base=@n68) {constExponent=20}, B=NamedRerouteUsage()), B=@n65)
  n70 = ScalarParameter() {defaultValue=2, group=ColorShading, parameterName=CoverColorIntensity_01}
  n71 = Multiply(A=Multiply(A=@n63.RGB, B=@n63.RGB), B=Add(A=Multiply(A=@n65, B=@n67), B=Multiply(A=@n69, B=@n70) {constB=2}))
  n72 = LinearInterpolate(A=ScalarParameter() {defaultValue=0.2, group=ColorShading, parameterName=OutsideAreaIntensity_01}, Alpha=NamedRerouteUsage())
  NamedRerouteDeclaration(None=Multiply(A=Multiply(A=@n71, B=@n72), B=NamedRerouteUsage())) {name=FirstLayerColor, nodeColor=<rgba 1 0 0.8438 1>}
  n73 = NamedRerouteUsage()
  n74 = VectorParameter() {group=ColorShading, parameterName=SecondLayerColor, defaultValue=<rgba 0.4427 0.7574 1 2>}
  n75 = Distance(A=NamedRerouteUsage(), B=Constant3Vector() {constant=<rgba 0 0 1 0>})
  n76 = Power(Base=@n75, Exp=ScalarParameter() {defaultValue=2.5, group=ColorShading, parameterName=SecondLayerIcePower})
  n77 = LinearInterpolate(
   A=ScalarParameter() {defaultValue=0.5, group=ColorShading, parameterName=BackSideIntensity_02}
   B=ScalarParameter() {defaultValue=1, group=ColorShading, parameterName=FrontSideIntensity_02}
   Alpha=NamedRerouteUsage()
  )
  n78 = Multiply(A=@n77, B=OneMinus(None=@n75))
  n79 = Multiply(A=Multiply(A=@n74.RGB, B=@n74.RGB), B=Add(A=Multiply(A=@n76, B=@n77), B=Multiply(A=@n78) {constB=0.1}))
  n80 = ScalarParameter() {defaultValue=3, group=ColorShading, parameterName=CoverColorIntensity_Second}
  n81 = NamedRerouteUsage()
  NamedRerouteDeclaration(None=Multiply(A=Multiply(A=@n73, B=@n79), B=LinearInterpolate(B=@n80, Alpha=@n81) {constA=1, constB=3})) {name=SecondLayerColor, nodeColor=<rgba 1 0.2109 0 1>}
  SkyAtmosphereLightDirection()
 ]}
